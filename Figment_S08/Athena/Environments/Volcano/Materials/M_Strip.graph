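# Translucent lit volcano material for an animated lava strip, blending a dark color into bright orange emissive.
# Three panned lava mask samples, one distorting another, combine with a UV1 gradient and contrast powers to drive glow and opacity.
# Vertex red masks the glow, vertex alpha masks opacity, and world position offset pushes vertices along normals and far along the camera ray.
# The graph also holds an object-relative edge-gradient branch through MF_LavaEdgeGradientSimple.
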Material /Figment_S08/Athena/Environments/Volcano/Materials/M_Strip
graph {
 properties {bUseMaterialAttributes=true, blendMode=BLEND_Translucent, pixelDepthOffsetMode=PDOM_Legacy, usedShadingModels=Default Lit}
 nodes [
  n1 = Multiply(A=TextureCoordinate(), B=ScalarParameter() {defaultValue=10.0182, parameterName=TilingLava})
  n2 = ScalarParameter() {defaultValue=-0.03, parameterName=Pan1}
  n3 = ScalarParameter() {defaultValue=-0.0286, parameterName=Distort}
  n4 = TextureCoordinate()
  n5 = ScalarParameter() {defaultValue=12.5006, parameterName=TilingLava2}
  n6 = ScalarParameter() {defaultValue=0.03, parameterName=Pan2}
  n7 = TextureSample(UVs=Panner(Coordinate=Multiply(A=@n4, B=@n5), Speed=@n6)) {samplerType=SAMPLERTYPE_Masks, texture=/Game/Environments/Sets/Vulcano/Textures/LavaGround_M}
  n8 = TextureSample(UVs=Add(A=Panner(Coordinate=@n1, Speed=@n2), B=Multiply(A=@n3, B=@n7.G))) {samplerType=SAMPLERTYPE_Masks, texture=&LavaGround_M}
  Add(A=@n8.G, B=@n7.G)
  n9 = Constant()
  AppendVector(A=@n6, B=@n9)
  AppendVector(A=@n2, B=@n9)
  CameraPositionWS()
  n10 = ObjectPositionWS()
  n11 = ComponentMask(None=Subtract(A=WorldPosition().XYZ, B=@n10)) {b=true}
  n12 = Divide(A=ComponentMask(None=@n10) {b=true}) {constB=384}
  n13 = Add(A=Multiply(A=Frac(None=@n12)) {constB=384}, B=ScalarParameter() {defaultValue=-306.7278, group=Lava, parameterName=EdgeGradientOffset})
  n14 = Clamp(None=Divide(A=Add(A=@n11, B=@n13), B=ScalarParameter() {defaultValue=54.307, group=Lava, parameterName=EdgeGradientLength}))
  n15 = MaterialFunctionCall!(
   In=MakeMaterialAttributes!(
    BaseColor
    Metallic
    Specular
    Roughness
    Anisotropy
    EmissiveColor
    Opacity
    OpacityMask
    Normal
    Tangent
    WorldPositionOffset
    SubsurfaceColor
    ClearCoat
    ClearCoatRoughness
    AmbientOcclusion
    Refraction
    CustomizedUVs_0
    CustomizedUVs_1
    CustomizedUVs_2
    CustomizedUVs_3
    CustomizedUVs_4
    CustomizedUVs_5
    CustomizedUVs_6
    CustomizedUVs_7
    PixelDepthOffset
    ShadingModel
    Displacement
   )
   UVs
   TransitionGradient=@n14
  ) {materialFunction=/Game/Environments/Sets/Vulcano/Materials/MF_LavaEdgeGradientSimple}
  n16 = GetMaterialAttributes(None=@n15.Result) {attributeGetTypes [#1, #2]}
  LinearInterpolate(A=@n16.Emissive Color, Alpha=Desaturation(None=@n16.Emissive Color))
  n17 = TextureCoordinate()
  n18 = ScalarParameter() {defaultValue=8.8147, parameterName=TilingLava3}
  n19 = TextureSample(UVs=Panner(Coordinate=Multiply(A=@n17, B=@n18), Speed=ScalarParameter() {defaultValue=0.0171, parameterName=Pan3})) {samplerType=SAMPLERTYPE_Masks, texture=/Game/Environments/Sets/Vulcano/Textures/Lava_M}
  n20 = OneMinus(None=Multiply(A=@n8.G, B=@n7.G))
  n21 = ComponentMask(None=TextureCoordinate() {coordinateIndex=1}) {g=true}
  n22 = Power(Base=Multiply(A=Multiply(A=@n19.R, B=@n20), B=@n21), Exp=ScalarParameter() {defaultValue=4.8349, parameterName=HeightContrast})
  n23 = Add(A=Power(Base=@n22, Exp=ScalarParameter() {defaultValue=1.3048, parameterName=HeightContrast2}), B=@n22)
  n24 = VertexColor()
  n25 = LinearInterpolate(
   A=ScalarParameter() {defaultValue=0.4937, parameterName=Opacity3}
   B=ScalarParameter() {defaultValue=0.6571, parameterName=Opacity4}
   Alpha=Multiply(A=@n19.R, B=@n21)
  )
  n26 = Add(
   A=@n22
   B=LinearInterpolate(
    A=ScalarParameter() {parameterName=Opacity1}
    B=ScalarParameter() {defaultValue=1.9566, parameterName=Opacity2}
    Alpha=Multiply(A=@n25, B=@n21)
   )
  )
  n27 = VertexColor()
  n28 = VertexNormalWS()
  n29 = ScalarParameter() {defaultValue=9.7919, parameterName=Offsetvert}
  n30 = Normalize(VectorInput=Subtract(A=WorldPosition() {worldPositionShaderOffset=WPT_ExcludeAllShaderOffsets}.XYZ, B=CameraPositionWS()))
  n31 = ScalarParameter() {defaultValue=5000, parameterName=Offset}
  MakeMaterialAttributes(
   EmissiveColor=LinearInterpolate(
    A=VectorParameter() {parameterName=Color, defaultValue=<rgba 0.03 0.0093 0.0021 1>}.RGB
    B=VectorParameter() {parameterName=Color2, defaultValue=<rgba 3 0.2076 0.0583 1>}.RGB
    Alpha=Multiply(A=@n23, B=@n24.R)
   )
   Opacity=Clamp(None=Multiply(A=@n26, B=@n27.A))
   WorldPositionOffset=Add(A=Multiply(A=@n28, B=@n29), B=Multiply(A=@n30, B=@n31))
  )
  MaterialFunctionCall(Transition Phase=@n21, Height Texture=@n22, Contrast=ScalarParameter() {defaultValue=-0.5766, parameterName=HeightlerpContrast}) {materialFunction=&HeightLerp}
  Max(A=@n8.G, B=@n7.G)
  n32 = ScalarParameter() {defaultValue=0.9403, parameterName=Opacity}
  Multiply(A=OneMinus(None=@n14), B=@n32)
  n33 = Subtract(A=WorldPosition().XYZ, B=CameraPositionWS())
  n34 = ScalarParameter() {defaultValue=1.9566, parameterName=Opacity2}
  SetMaterialAttributes!(
   MaterialAttributes=@n15.Result
   World Position Offset=Multiply(A=Normalize(VectorInput=@n33), B=ScalarParameter() {defaultValue=5000, parameterName=Offset})
   Opacity=Clamp(None=LinearInterpolate(A=@n32, B=@n34, Alpha=@n14))
  ) {attributeSetTypes [#3, #4]}
 ]
}
