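# Default Lit material for the volcano lava throne, pairing its diffuse, normal and mask textures with a computed orange-red glow.
# The mask's alpha is sampled sharp and at a blurred mip, summed for a soft bloom, boosted, then offset by a world-X gradient from the actor position.
# A color parameter tints this and feeds an added dimmer glow that fades with height.
# A feature-level switch zeroes the glow on ES2 and ES3_1.
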Material /Figment_S08/Athena/Environments/Volcano/Materials/M_ThroneGlow
graph {
 properties {pixelDepthOffsetMode=PDOM_Legacy, usedShadingModels=Default Lit}
 nodes [
  n1 = Constant() {r=0.1714}
  n2 = TextureSample!(UVs, Tex, Level=Constant() {r=6}, Apply View MipBias) {
   constMipValue=7
   mipValueMode=TMVM_MipLevel
   samplerSource=SSM_Clamp_WorldGroupSettings
   samplerType=SAMPLERTYPE_Masks
   texture=/Game/Athena/Environments/Volcano/Textures/T_Volcano_LavaThrone_S
  }
  n3 = TextureSample!(UVs, Tex, Apply View MipBias) {
   samplerSource=SSM_Clamp_WorldGroupSettings
   samplerType=SAMPLERTYPE_Masks
   texture=&T_Volcano_LavaThrone_S
  }
  n4 = Multiply(A=Add(A=Multiply(A=@n1, B=@n2.A), B=@n3.A), B=Constant() {r=22.6156})
  n5 = ComponentMask(None=Subtract(A=WorldPosition().XYZ, B=ActorPositionWS())) {r=true}
  n6 = Constant() {r=52.4087}
  n7 = Saturate(None=Divide(A=Add(A=@n5, B=@n6), B=Constant() {r=43.3463}))
  n8 = VectorParameter() {parameterName=Param_1, defaultValue=<rgba 1.25 0.1039 0 0>}
  n9 = Constant() {r=0.381}
  n10 = WorldPosition()
  n11 = ActorPositionWS()
  n12 = Divide(A=ComponentMask(None=Subtract(A=@n10.XYZ, B=@n11)) {b=true}, B=Constant() {r=67.2143})
  n13 = Add(A=Multiply(A=Add(A=@n4, B=@n7), B=@n8.RGB), B=LinearInterpolate(A=Multiply(A=@n8.RGB, B=@n9), Alpha=Saturate(None=@n12)) {constB=0})
  n14 = Constant()
  FeatureLevelSwitch(Default=@n13, ES2=@n14, ES3_1=@n14, SM4_REMOVED=@n13)
  TextureSample() {texture=/Game/Athena/Environments/Volcano/Textures/T_Volcano_LavaThrone_D}
  TextureSample() {
   samplerSource=SSM_Clamp_WorldGroupSettings
   samplerType=SAMPLERTYPE_Normal
   texture=/Game/Athena/Environments/Volcano/Textures/T_Volcano_LavaThrone_N
  }
  TextureSample() {
   samplerSource=SSM_Clamp_WorldGroupSettings
   samplerType=SAMPLERTYPE_Masks
   texture=&T_Volcano_LavaThrone_S
  }
 ]
}
