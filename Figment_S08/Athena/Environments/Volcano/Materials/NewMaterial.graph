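# Translucent deferred decal material in the Volcano environment folder, apparently an unfinished draft of a cave rock wall decal that blends with terrain color.
# It multiplies a UV-scaled cave wall diffuse by the Athena terrain colormap, sampled through terrain-coordinate UVs, and by the landscape collection's ColormapDivisor.
# A matching normal sample shares the tiling.
# No material outputs appear connected, and the square mask texture, constant and two scalar parameters are unused.
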Material /Figment_S08/Athena/Environments/Volcano/Materials/NewMaterial
graph {
 properties {blendMode=BLEND_Translucent, materialDomain=MD_DeferredDecal, pixelDepthOffsetMode=PDOM_Legacy, usedShadingModels=Default Lit}
 nodes [
  Constant()
  n1 = Multiply(A=TextureCoordinate(), B=ComponentMask(None=VectorParameter() {defaultValue=<rgba 0.5114 1.1697 0 0>}.RGB) {g=true, r=true})
  n2 = TextureObject() {texture=/Game/Packages/DS_Fortnite_Terrain_NoLOD/Textures/T_Cave_Exterior_Wall_D}
  n3 = MaterialFunctionCall!(WorldPosition) {materialFunction=/Game/Athena/Environments/Landscape/MaterialFunctions/Utilities/MF_Athena_Terrain_Coords}
  Multiply(
   A=Multiply(
    A=TextureSample(UVs=@n1, Tex=@n2).RGB
    B=TextureSample(UVs=@n3.Result) {samplerType=SAMPLERTYPE_LinearColor, texture=/Game/Athena/Environments/Landscape/Textures/T_Athena_Terrain_CombinedColors_01}.RGB
   )
   B=CollectionParameter() {collection=/Game/Athena/Environments/Landscape/MPC/MPC_Landscape, parameterName=ColormapDivisor}
  )
  ScalarParameter() {defaultValue=0.1, parameterName=Param_1}
  ScalarParameter() {defaultValue=0.9, parameterName=Param_2}
  TextureSample() {texture=/Game/Athena/HUD/Materials/Score/T_UI_Square_DF}
  TextureSample(
   UVs=@n1
   Tex=TextureObject() {samplerType=SAMPLERTYPE_Normal, texture=/Game/Packages/DS_Fortnite_Terrain_NoLOD/Textures/T_Cave_Exterior_Wall_N}
  )
 ]
}
